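# Physics asset for the Buff Cat Comic cosmetic backpack, defining collision bodies and joints for the pack and its dyn_cap and dyn_fish bones.
# Boxes cover the pack and cap, while capsules cover the fish and a tiny root body.
# Constraints form a small tree, with pack joined to root and both cap and fish joined to pack.
# The cap joint locks twist and one swing axis, leaving hinge-like motion on the other swing axis.
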PhysicsAsset /BRCosmetics/Accessories/FORT_Backpacks/Backpack_M_MED_Buff_Cat_Comic/Meshes/M_MED_Buff_Cat_Comic_Pack_Physics
bodies {
 dyn_cap [{shapeType=Box, extentX=22, extentY=25, extentZ=3, center=<10.663 -0.0023 0.5373>}]
 dyn_fish [{shapeType=Capsule, length=8.6319, radius=6.3835, center=<4.0348 -0.2002 1.073>, rotation=<rot 90 0 0>}]
 pack [{shapeType=Box, extentX=23.9353, extentY=23.8094, extentZ=5, center=<0.1578 -10.9266 -0.0391>}]
 root [{shapeType=Capsule, length=0.505, radius=0.505, rotation=<rot 0 -90 0>}]
}
constraints [
 {bone1Name=dyn_cap, bone2Name=pack, swing1LimitDegrees=1, swing1Motion=Locked, swing2LimitDegrees=10, twistLimitDegrees=1, twistMotion=Locked}
 {bone1Name=pack, bone2Name=root, swing1LimitDegrees=1, swing2LimitDegrees=1, twistLimitDegrees=5}
 {bone1Name=dyn_fish, bone2Name=pack, swing1LimitDegrees=5, swing2LimitDegrees=10, twistLimitDegrees=2}
]
