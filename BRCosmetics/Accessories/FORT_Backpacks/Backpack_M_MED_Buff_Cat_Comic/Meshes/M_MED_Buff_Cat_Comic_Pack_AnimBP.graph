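# Animation Blueprint for the Buff Cat Comic backpack cosmetic, adding simulated physics motion to the pack skeleton.
# The local ref pose passes through two chained RigidBody nodes in root base-bone space, toggled oppositely by bIsInFrontEnd.
# The front-end node uses its own physics asset and full component motion scales; the other uses the standard asset with reduced scales and clamped applied acceleration.
# The event graph holds only unconnected stub nodes.
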AnimBlueprint /BRCosmetics/Accessories/FORT_Backpacks/Backpack_M_MED_Buff_Cat_Comic/Meshes/M_MED_Buff_Cat_Comic_Pack_AnimBP
target_skeleton=&/M_MED_Buff_Cat_Comic_Pack_Skeleton
graphs [
 AnimGraph {nodes [
   n1 = Get bIsInFrontEnd()
   n2 = RigidBody!(
    ComponentPose=Local To Component(LocalPose=Local Space Ref Pose())
    bAlphaBoolEnabled:bool=@n1
    OverrideWorldGravity:Vector="0,0,-980"
    ExternalForce
   ) -> (Pose) {
    alphaInputType=Bool
    bClampLinearTranslationLimitToRefPose=true
    bOverrideWorldGravity=true
    overridePhysicsAsset=&/M_MED_Buff_Cat_Comic_Pack_Physics_FrontEnd
    simulationSpace=BaseBoneSpace
    alphaBoolBlend {blendOption=Linear}
    baseBoneRef {boneName=root}
    componentLinearAccScale=<1 1 1>
    componentLinearVelScale=<1 1 1>
    overrideWorldGravity {z=-980}
    simSpaceSettings {
     maxAngularAcceleration=1000
     maxAngularVelocity=3
     maxLinearAcceleration=700
     maxLinearVelocity=500
     velocityScaleZ=0
     worldAlpha=1
     externalLinearDragV=<1 1>
    }
   }
   n3 = RigidBody(ComponentPose=@n2, bAlphaBoolEnabled=NOT Boolean(A=@n1), OverrideWorldGravity="0,0,-980") {
    alphaInputType=Bool
    bClampLinearTranslationLimitToRefPose=true
    bOverrideWorldGravity=true
    overridePhysicsAsset=&/M_MED_Buff_Cat_Comic_Pack_Physics
    simulationSpace=BaseBoneSpace
    alphaBoolBlend {blendOption=Linear}
    baseBoneRef {boneName=root}
    componentAppliedLinearAccClamp=<3500 3500 3500>
    componentLinearAccScale=<0.4 0.4 0.5>
    componentLinearVelScale=<0.4 0.4 0.5>
    overrideWorldGravity {z=-980}
    simSpaceSettings {
     maxAngularAcceleration=1000
     maxAngularVelocity=3
     maxLinearAcceleration=700
     maxLinearVelocity=500
     velocityScaleZ=0
     worldAlpha=1
     externalLinearDragV=<1 1>
    }
   }
   Output Pose(Result=Component To Local(ComponentPose=@n3)) {layerGroup=None}
  ]}
 EventGraph {nodes [
   TryGetPawnOwner()
   Event BlueprintUpdateAnimation()
  ]}
]
variables []
